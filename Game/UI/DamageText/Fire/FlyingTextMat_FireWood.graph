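# Masked unlit material for fire-styled flying damage numbers, named for the FireWood variant.
# UV2 red blends digit and plus/minus atlases for distance-field opacity, normals and packed masks.
# DamageNumberWPO supplies animation, time, highlight and critical-hit data.
# Panned noise, edge masks and fresnel drive an orange glow over fake skylight-tinted cubemap diffuse, scaled by time of day, boosted on crits and dimmed over the animation.
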Material /Game/UI/DamageText/Fire/FlyingTextMat_FireWood
graph {
 properties {
  bEnableResponsiveAA=true
  blendMode=BLEND_Masked
  numCustomizedUVs=5
  pixelDepthOffsetMode=PDOM_Legacy
  previewMesh=/Game/UI/DamageText/Default/standard_DamageNumbers8
  shadingModel=MSM_Unlit
  translucencyLightingMode=TLM_Surface
  usedShadingModels=Unlit
 }
 nodes [
  Constant() {r=0.5}
  n1 = TextureCoordinate() {coordinateIndex=1}
  n2 = Add(A=TextureCoordinate() {vTiling=0.5}, B=AppendVector(A=Constant(), B=ScalarParameter() {parameterName=+Or-}))
  n3 = ComponentMask(None=TextureCoordinate() {coordinateIndex=2}) {r=true}
  LinearInterpolate(
   A=TextureSample(UVs=@n1) {texture=&/FireNumbers_opacity_g_distancefield}.R
   B=TextureSample(UVs=@n2) {texture=&/FireNumbers_opacity_g_distancefield_+-.FireNumbers_opacity_g_distancefield_+-}.R
   Alpha=@n3
  )
  n4 = LinearInterpolate(
   A=TextureSampleParameter2D(UVs=@n1) {group=Visuals, parameterName=Normal, samplerType=SAMPLERTYPE_Normal, texture=&/FireNormalsMap}.RGB
   B=TextureSample(UVs=@n2) {samplerType=SAMPLERTYPE_Normal, texture=&/Fire+-Normal.Fire+-Normal}.RGB
   Alpha=@n3
  )
  n5 = Constant3Vector() {constant=<rgba 1 1 0.5 0>}
  n6 = MaterialFunctionCall!(
   Normals=Normalize(VectorInput=Multiply(A=@n4, B=@n5))
   Debug texture coordinate input
   Animation
   Additonal Scale
   "Additional Camera Offset "
   "Additonal Screenspace XY offset "
  ) {materialFunction=/Game/UI/DamageText/Base/DamageNumberWPO}
  n7 = Constant() {r=1}
  n8 = MaterialFunctionCall(Color_Morning=@n7, Color_Day=@n7, Color_Evening=@n7, Color_Night=Constant() {r=0.615}) {materialFunction=&Color_Time_of_Day}
  n9 = TextureSample(UVs=@n1) {desc=DamageNumbers_Ice_RmistCore_G_Edges_B_AO, texture=&/DamageNumbers_Fire_R_brightEdges_G_Crevasses_B_AO}
  n10 = TextureSample(UVs=@n2) {
   desc=DamageNumbers_Ice_RmistCore_G_Edges_B_AO
   texture=&/DamageNumbers_Fire_R_brightEdges_G_Crevasses_B_AO_+-.DamageNumbers_Fire_R_brightEdges_G_Crevasses_B_AO_+-
  }
  n11 = LinearInterpolate(A=AppendVector(A=@n9.RGB, B=@n9.RGB), B=AppendVector(A=@n10.RGB, B=@n10.RGB), Alpha=@n3)
  n12 = ComponentMask(None=@n11) {a=true}
  n13 = Constant() {r=2}
  n14 = MaterialFunctionCall(Float2=TextureCoordinate() {coordinateIndex=4}) {materialFunction=&BreakOutFloat2Components}
  n15 = ComponentMask(None=@n4) {g=true, r=true}
  n16 = Multiply(A=Add(A=Add(A=@n1, B=@n14.R), B=Multiply(A=@n15) {constB=0.01}), B=Constant2Vector() {g=10.1235, r=1})
  n17 = LinearInterpolate(Alpha=TextureSample(UVs=Panner(Coordinate=@n16) {speedY=0.2}) {texture=&T_NoisePack_01}.R) {constA=0.4}
  n18 = Multiply(
   A=Multiply(
    A=MaterialFunctionCall(In=@n12, Contrast=@n13) {materialFunction=&CheapContrast}.Result
    B=@n6.isCriticalHit?
   ) {constB=3}
   B=@n17
  )
  n19 = MaterialFunctionCall(Float3=ComponentMask(None=@n11) {b=true, g=true, r=true}) {
   desc=DamageNumbers_Ice_RmistCore_G_Edges_B_AO
   materialFunction=&BreakOutFloat3Components
  }
  n20 = Transform(None=@n6.Normals)
  n21 = Fresnel(Normal=@n20) {baseReflectFraction=0, exponent=1.5465}
  n22 = Multiply(A=@n17, B=Multiply(A=@n19.G, B=Add(A=@n19.R, B=@n21)))
  n23 = LinearInterpolate(B=Constant3Vector() {constant=<rgba 2 1.1387 0.3623 0>}, Alpha=Clamp(None=Add(A=@n18, B=@n22)))
  n24 = CollectionParameter() {
   collection=&FortniteMaterialParameters
   parameterName=SkyLightColorAndIntensity
  }
  n25 = TextureSampleParameterCube!(UVs=ReflectionVectorWS(CustomWorldNormal=@n20), Bias=Constant() {r=100}, Apply View MipBias) {mipValueMode=TMVM_MipBias, parameterName=Cubemap, texture=/Game/Packages/Fortress_Sky/TexturesHDR/T_hdrsky_test03_low}
  n26 = Power(Base=@n19.B) {constExponent=1}
  n27 = Multiply(
   A=Constant3Vector() {desc=diffuse color, constant=<rgba 0.0413 0.043 0.0363 0>}
   B=Multiply(A=ComponentMask(None=@n24) {b=true, g=true, r=true}, B=Multiply(A=@n25.RGB, B=@n26))
  )
  n28 = Multiply(A=Add(A=Multiply(A=@n23, B=@n23), B=Multiply(A=@n19.B, B=@n27)), B=@n6.Pixel shader Letter Highlight - uv 3)
  Multiply(
   A=OneMinus(None=@n6.Time)
   B=MaterialFunctionCall!(Base=Multiply(A=@n8.Result, B=@n28), Add=Multiply(A=@n6.isCriticalHit?) {constB=0.5, desc=Crit multiplier}) {materialFunction=/Engine/Functions/Engine_MaterialFunctions02/Math/MultiplyAdd}.Result
  )
 ]
}
