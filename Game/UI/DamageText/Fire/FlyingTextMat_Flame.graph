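# Translucent unlit material for flaming damage-number text, animated through the shared DamageNumberWPO function.
# A panning flow map with noise-jittered time distorts two blended phases of digit and plus/minus atlases, selected by UV2.
# Sphere masks around a tenths-quantized UV5 value isolate each glyph's atlas band.
# The contrasted mask drives depth-faded opacity and squared fire-color emissive, scaled by time of day and letter highlight, fading over the animation.
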Material /Game/UI/DamageText/Fire/FlyingTextMat_Flame
graph {
 properties {
  bEnableResponsiveAA=true
  blendMode=BLEND_Translucent
  numCustomizedUVs=6
  pixelDepthOffsetMode=PDOM_Legacy
  previewMesh=/Game/UI/DamageText/Default/standard_DamageNumbers8
  shadingModel=MSM_Unlit
  translucencyLightingMode=TLM_Surface
  usedShadingModels=Unlit
 }
 nodes [
  Constant() {r=0.5}
  Constant() {r=0.5}
  n1 = MaterialFunctionCall(Speed=Constant() {r=1}) {materialFunction=&TimeWithSpeedVariable}
  n2 = Add(A=TextureCoordinate(), B=Multiply(A=@n1.Time, B=Constant() {r=0.02}))
  Frac(None=@n2)
  n3 = Add(A=MaterialFunctionCall(Float2=TextureCoordinate() {coordinateIndex=5}) {materialFunction=&BreakOutFloat2Components}.G) {constB=0.5}
  MaterialFunctionCall(Number=Divide(A=Floor(None=@n3)) {constB=4}) {materialFunction=&DebugScalarValues}
  n4 = Multiply(A=MaterialFunctionCall(Float2=TextureCoordinate() {coordinateIndex=5}) {materialFunction=&BreakOutFloat2Components}.R) {constB=10}
  n5 = Divide(A=Floor(None=Add(A=@n4) {constB=0.5})) {constB=10}
  MaterialFunctionCall(Number=@n5) {materialFunction=&DebugScalarValues}
  n6 = Panner(Coordinate=TextureCoordinate() {coordinateIndex=1}) {speedY=0.1}
  n7 = TextureCoordinate() {coordinateIndex=5}
  n8 = TextureSample!(UVs=@n2, Tex, Level, Apply View MipBias) {
   desc=noise to break up the repetition
   mipValueMode=TMVM_MipLevel
   samplerType=SAMPLERTYPE_LinearGrayscale
   texture=/Engine/Functions/Engine_MaterialFunctions02/ExampleContent/Textures/LowResBlurredNoise
  }
  n9 = MaterialFunctionCall(Float2=TextureCoordinate() {coordinateIndex=4}) {materialFunction=&BreakOutFloat2Components}
  n10 = Multiply(A=@n9.R, B=ScalarParameter() {defaultValue=-0.1, parameterName=X Fire Offset})
  n11 = Multiply(A=@n9.G, B=ScalarParameter() {defaultValue=-0.0087, parameterName=Y Fire Offset})
  n12 = MaterialFunctionCall!(
   Diffuse
   Normal Map
   "Flow Vector Map (see tooltip)"=ConstantBiasScale(None=TextureSample!(UVs=@n6, Tex, Apply View MipBias) {samplerType=SAMPLERTYPE_LinearColor, texture=&/DamageNumbers_Fire_RG_FlowMap}.RGB) {bias=-0.5, scale=4}
   Green UV Target Number=Add(A=MaterialFunctionCall(Float2=@n7) {materialFunction=&BreakOutFloat2Components}.R) {constB=0.05}
   UV texture 2 offset=Constant()
   "Time (see tooltip)"=Add(A=Multiply(A=@n8.R) {constB=0.25}, B=@n1.Time)
   useMipLevel
   UVs=Add(A=TextureCoordinate() {coordinateIndex=1}, B=AppendVector(A=@n10, B=@n11))
   Texture Size
   Texture Mip Bias
  ) {materialFunction=/Game/Packages/Fortnite_UniqueMaterials_1/MaterialFunctions/FlowMaps_additionalOutputs}
  MaterialFunctionCall(Float2=@n12.UVA) {materialFunction=&BreakOutFloat2Components}
  MaterialFunctionCall(Speed=ScalarParameter() {defaultValue=0.1, parameterName=debug time}) {materialFunction=&TimeWithSpeedVariable}
  n13 = Add(A=@n5) {constB=0.05}
  n14 = SphereMask(
   A=MaterialFunctionCall(Float2=@n12.UVA) {materialFunction=&BreakOutFloat2Components}.G
   B=@n13
  ) {attenuationRadius=0.05}
  n15 = Multiply(A=TextureSample(UVs=@n12.UVA) {texture=&/FireNumbers_opacity_g_distancefield}.RGB, B=@n14)
  n16 = SphereMask(
   A=MaterialFunctionCall(Float2=@n12.UVB) {materialFunction=&BreakOutFloat2Components}.G
   B=@n13
  ) {attenuationRadius=0.05}
  n17 = Multiply(A=TextureSample(UVs=@n12.UVB) {texture=&FireNumbers_opacity_g_distancefield}.RGB, B=@n16)
  n18 = Multiply(
   A=TextureSample(UVs=@n12.UVA) {texture=&/FireNumbers_opacity_g_distancefield_+-.FireNumbers_opacity_g_distancefield_+-}.RGB
   B=@n14
  )
  n19 = Multiply(
   A=TextureSample(UVs=@n12.UVB) {texture=&FireNumbers_opacity_g_distancefield_+-.FireNumbers_opacity_g_distancefield_+-}.RGB
   B=@n16
  )
  n20 = TextureCoordinate() {coordinateIndex=2}
  n21 = ComponentMask(None=LinearInterpolate(
    A=LinearInterpolate(A=@n15, B=@n17, Alpha=@n12.Lerp Alpha)
    B=LinearInterpolate(A=@n18, B=@n19, Alpha=@n12.Lerp Alpha)
    Alpha=ComponentMask(None=@n20) {r=true}
   )) {g=true}
  n22 = MaterialFunctionCall(Float2=TextureCoordinate()) {materialFunction=&BreakOutFloat2Components}
  n23 = Clamp(None=Multiply(A=ScalarParameter() {defaultValue=5.7484, parameterName=GMultiplier}, B=@n22.G))
  Multiply(A=@n21, B=@n23)
  n24 = Multiply(
   A=MaterialFunctionCall(In=@n21, Contrast=Constant() {r=10}) {materialFunction=&CheapContrast}.Result
   B=@n23
  )
  Multiply(
   A=DepthFade(Opacity=@n24, FadeDistance=ScalarParameter() {defaultValue=4256.1025, parameterName=Distance Fade}) {fadeDistanceDefault=50}
   B=ScalarParameter() {defaultValue=0.1881, parameterName=Opacity}
  ) {constB=0.5, desc=OPACITY}
  n25 = Constant() {r=0.5}
  n26 = ScalarParameter() {defaultValue=0.918, parameterName=hardness}
  Multiply(
   A=SphereMask(A=@n22.R, B=@n25, Hardness=@n26) {attenuationRadius=0.5, hardnessPercent=10}
   B=SphereMask(A=@n22.G, B=@n25, Hardness=@n26) {attenuationRadius=0.5, hardnessPercent=50}
  )
  n27 = Constant() {r=1}
  n28 = MaterialFunctionCall(Color_Morning=@n27, Color_Day=@n27, Color_Evening=@n27, Color_Night=Constant() {r=0.615}) {materialFunction=&Color_Time_of_Day}
  n29 = LinearInterpolate(B=VectorParameter() {parameterName=Fire Color, defaultValue=<rgba 4 2.2773 0.7246 1>}.RGB, Alpha=@n24)
  n30 = Multiply(A=@n29, B=@n29)
  n31 = MaterialFunctionCall!(
   Normals
   Debug texture coordinate input
   Animation
   Additonal Scale=Multiply(A=Constant2Vector() {g=1.2221, r=1.4442}, B=ScalarParameter() {defaultValue=0.9366, parameterName=Additional Scale}) {constB=1.1115}
   "Additional Camera Offset "=ScalarParameter() {defaultValue=10, parameterName=Camera Offset}
   "Additonal Screenspace XY offset "
  ) {materialFunction=/Game/UI/DamageText/Base/DamageNumberWPO}
  Multiply(A=Multiply(A=Multiply(A=@n28.Result, B=@n30), B=@n31.Pixel shader Letter Highlight - uv 3) {desc=eMISSIVE}, B=OneMinus(None=@n31.Time))
  TextureSample() {samplerType=SAMPLERTYPE_LinearColor, texture=&/DamageNumbers_Fire_R_brightEdges_G_Crevasses_B_AO}
  TextureSample() {samplerType=SAMPLERTYPE_LinearColor, texture=&DamageNumbers_Fire_RG_FlowMap}
  TextureSample() {samplerType=SAMPLERTYPE_Normal, texture=&/FireNormalsMap}
 ]
}
