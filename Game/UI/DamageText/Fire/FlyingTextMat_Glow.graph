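# Translucent unlit glow material for fire-styled flying damage-number text.
# UV channel 2 blends between digit and plus/minus distance-field atlases, with a parameter offsetting the sign lookup.
# The distance field drives a squared Fire Color gradient, scaled by time of day and by letter highlight and time fade from the shared DamageNumberWPO function.
# Two high-contrast thresholds of the inverted field are averaged and inverted into a stepped mask.
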Material /Game/UI/DamageText/Fire/FlyingTextMat_Glow
graph {
 properties {
  bEnableResponsiveAA=true
  blendMode=BLEND_Translucent
  numCustomizedUVs=5
  pixelDepthOffsetMode=PDOM_Legacy
  previewMesh=/Game/UI/DamageText/Default/standard_DamageNumbers8
  shadingModel=MSM_Unlit
  translucencyLightingMode=TLM_Surface
  usedShadingModels=Unlit
 }
 nodes [
  MaterialFunctionCall(Speed=ScalarParameter() {defaultValue=0.1, parameterName=debug time}) {materialFunction=&TimeWithSpeedVariable}
  n1 = Constant() {r=1}
  n2 = MaterialFunctionCall(Color_Morning=@n1, Color_Day=@n1, Color_Evening=@n1, Color_Night=Constant() {r=0.615}) {materialFunction=&Color_Time_of_Day}
  n3 = TextureCoordinate() {coordinateIndex=1}
  n4 = Add(A=TextureCoordinate() {vTiling=0.5}, B=AppendVector(A=Constant(), B=ScalarParameter() {parameterName=+Or-}))
  n5 = TextureCoordinate() {coordinateIndex=2}
  n6 = MaterialFunctionCall(Float3=LinearInterpolate(
    A=TextureSample(UVs=@n3) {texture=&/FireNumbers_opacity_g_distancefield}.G
    B=TextureSample(UVs=@n4) {texture=&/FireNumbers_opacity_g_distancefield_+-.FireNumbers_opacity_g_distancefield_+-}.RGB
    Alpha=ComponentMask(None=@n5) {r=true}
   )) {desc=R opacity mask G Distance field, materialFunction=&BreakOutFloat3Components}
  n7 = LinearInterpolate(B=VectorParameter() {parameterName=Fire Color, defaultValue=<rgba 4 2.2773 0.7246 1>}.RGB, Alpha=@n6.G)
  n8 = Multiply(A=@n7, B=@n7)
  n9 = MaterialFunctionCall!(
   Normals
   Debug texture coordinate input
   Animation
   Additonal Scale=Constant2Vector() {g=1.1, r=1.3}
   "Additional Camera Offset "=ScalarParameter() {defaultValue=-5, parameterName=Camera Offset}
   "Additonal Screenspace XY offset "
  ) {materialFunction=/Game/UI/DamageText/Base/DamageNumberWPO}
  Multiply(A=Multiply(A=Multiply(A=@n2.Result, B=@n8), B=@n9.Pixel shader Letter Highlight - uv 3), B=OneMinus(None=@n9.Time))
  n10 = OneMinus(None=@n6.G)
  n11 = Constant() {r=40}
  n12 = MaterialFunctionCall(In=@n10, Contrast=@n11) {materialFunction=&CheapContrast}
  n13 = MaterialFunctionCall(In=Add(A=@n10) {constB=-0.3044}, Contrast=@n11) {materialFunction=&CheapContrast}
  OneMinus(None=Add(A=Multiply(A=@n12.Result) {constB=0.5}, B=Multiply(A=@n13.Result) {constB=0.5}))
  ScalarParameter() {defaultValue=0.5, parameterName=debug animation}
 ]
}
